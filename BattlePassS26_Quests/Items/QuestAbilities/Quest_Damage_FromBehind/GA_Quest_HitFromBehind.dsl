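; Quest gameplay ability that detects when a player hits another Athena player pawn from behind with an assault rifle or SMG.
; On its triggering event, it filters on damage and weapon tags, then compares target and instigator forward vectors by dot product, treating similar facing as a back hit without a position check.
; A hit broadcasts a HitFromBehind custom verb message carrying damage done and sends a matching gameplay event; every branch ends the ability.
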(event Ability|EventActivateAbilityFromEvent
  (params (EventData :type Gameplay%20Event%20Data%20Structure%20%28by%20ref%29))
  (bind (_eventtag _instigator _target _optionalobject _optionalobject2 _contexthandle _instigatortags _targettags _eventmagnitude) (Utilities|Struct|BreakGameplayEventData EventData))
  (|CallParentFunction EventData)
  (Variables|Default|SetInstigator (Game|Player|GetPlayerPawn (Fortnite|GetFortPlayerControllerfromActor _instigator)))
  (Variables|Default|SetTarget _target)
  (if (GameplayTags|HasTag _instigatortags "(TagName=\"Gameplay.Damage\")")
    (if (or (GameplayTags|HasTag _instigatortags "(TagName=\"Item.Weapon.Ranged.Assault\")") (GameplayTags|HasTag _instigatortags "(TagName=\"Item.Weapon.Ranged.SMG\")"))
      (if (Utilities|ClassIsChildOf (Utilities|GetClass (Variables|Default|GetTarget)) "/Game/Athena/PlayerPawn_Athena.PlayerPawn_Athena_C")
        (if (>= (Math|Vector|DotProduct (Transformation|GetActorForwardVector (Variables|Default|GetTarget)) (Transformation|GetActorForwardVector (Variables|Default|GetInstigator))) 0.425)
          (bind _result (CustomVerb|AddTagsforSubject (CustomVerb|MakeCustomVerbMessage (Variables|GAVars|Data|GetOwningPlayerPawn) "(TagName=\"CustomVerb.HitFromBehind\")") "BackHit" (GameplayTags|MakeGameplayTagContainerfromTag)))
          (bind _result_2 (CustomVerb|AddValue _result "DamageDone" _eventmagnitude))
          (CustomVerb|BroadcastCustomVerbMessage _result_2)
          (Ability|SendGameplayEventtoActor 0 "(TagName=\"CustomVerb.HitFromBehind\")")
          (Ability|EndAbility)
          (else
            (Ability|EndAbility)))
        (else
          (Ability|EndAbility)))
      (else
        (Ability|EndAbility)))
    (else
      (Ability|EndAbility))))
